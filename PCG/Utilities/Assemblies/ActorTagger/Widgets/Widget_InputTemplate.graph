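# Editor Utility Widget template for an input row in the PCG assemblies ActorTagger tool, apparently a base for tag-input widgets.
# Its layout is an empty dark frame: a Border wrapping a fixed-size SizeBox and an inner Border.
# FormatTagValue takes a tag and an actor reference and returns the tag unchanged as a default pass-through.
# CleanUp and the PreConstruct, Construct and Tick events are empty stubs.
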EditorUtilityWidgetBlueprint /PCG/Utilities/Assemblies/ActorTagger/Widgets/Widget_InputTemplate
graphs [
 CleanUp {nodes [CleanUp() {is=title K2Node_FunctionEntry}]}
 EventGraph {nodes [
   Event PreConstruct()
   Event Construct()
   Event Tick()
  ]}
 FormatTagValue {nodes [
   FormatTagValue!() -> (then, Tag:String, ActorObjectReference:Actor Object Reference) {is=title K2Node_FunctionEntry}
   > Return Node!(execute, TagResult:String=Get Tag() -> (Tag:String))
  ]}
]
variables []
widgets {
 info {rootWidgetClass=&Border, widgetCount=3}
 widgets [
  Border_1 {
   widgetClassPath=&Border
   widget=Border_1
   properties {
    verticalAlignment=VAlign_Center
    background {
     bIsDynamicallyLoaded=false
     drawAs=Box
     imageType=NoImage
     mirroring=NoMirror
     resourceName=None
     resourceObject=null
     tiling=NoTile
     imageSize=<32 32>
     margin=<ltrb 0 0 0 0>
     outlineSettings {
      bUseBrushTransparency=false
      roundingType=HalfHeightRadius
      width=0
      color {colorUseRule=UseColor_Specified, specifiedColor=<rgba 0 0 0 0>}
      cornerRadii=<0 0 0 0>
     }
     tintColor {colorUseRule=UseColor_Specified, specifiedColor=<rgba 0.0176 0.0176 0.0176 1>}
     uVRegion {bIsValid=false, max=<0 0>, min=<0 0>}
    }
    padding=<ltrb 4 2 2 2>
   }
  }
  SizeBox_0 {
   widgetClassPath=&SizeBox
   parent=Border_1
   widget=SizeBox_0
   properties {bOverride_HeightOverride=true, bOverride_WidthOverride=true, heightOverride=54, widthOverride=485}
   slot {class=BorderSlot, properties {verticalAlignment=VAlign_Center, padding=<ltrb 4 2 2 2>}}
  }
  Border_0 {
   widgetClassPath=&Border
   parent=SizeBox_0
   widget=Border_0
   properties {
    verticalAlignment=VAlign_Center
    background {
     bIsDynamicallyLoaded=false
     drawAs=Box
     imageType=NoImage
     mirroring=NoMirror
     resourceName=None
     resourceObject=null
     tiling=NoTile
     imageSize=<32 32>
     margin=<ltrb 0 0 0 0>
     outlineSettings {
      bUseBrushTransparency=false
      roundingType=HalfHeightRadius
      width=0
      color {colorUseRule=UseColor_Specified, specifiedColor=<rgba 0 0 0 0>}
      cornerRadii=<0 0 0 0>
     }
     tintColor {colorUseRule=UseColor_Specified, specifiedColor=<rgba 0.0176 0.0176 0.0176 1>}
     uVRegion {bIsValid=false, max=<0 0>, min=<0 0>}
    }
    padding=<ltrb 2 2 2 2>
   }
   slot {class=SizeBoxSlot}
  }
 ]
}
