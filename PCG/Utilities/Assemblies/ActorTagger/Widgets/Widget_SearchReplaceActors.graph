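# Editor Utility Widget that bulk-replaces actors by source/target class or remap DataTable, which also matches StaticMeshActors by mesh.
# It processes levels selected in the Content Browser, falling back to selected Level Instances, and a second button replaces selected actors.
# Bulk replacement spawns targets at original transforms, copies tags, reattaches parents and children, then destroys originals.
# Changed levels are saved and exported to "_PCG" data assets via PCGLevelToAsset.
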EditorUtilityWidgetBlueprint /PCG/Utilities/Assemblies/ActorTagger/Widgets/Widget_SearchReplaceActors
graphs [
 CopyActorTags {nodes [
   n1 = CopyActorTags!() -> (then, ActorReference:Actor Object Reference) {is=title K2Node_FunctionEntry}
   > Clear(TargetArray:Array of Names=Get CopyPasteTags() -> (CopyPasteTags:Array of Names))
   > n2 = For Each Loop(Array:Array of Names=Get Tags!(self:Actor Object Reference=@n1) -> (Tags:Array of Names)) -> (Array Element:Name)
   > Add Unique(TargetArray:Array of Names=Get CopyPasteTags(), NewItem:"Name (by ref)"=@n2)
  ]}
 CopySelectedActorTags {nodes [
   CopySelectedActorTags!() -> (then, UsePredefinedTags:bool=true) {is=title K2Node_FunctionEntry}
   > n1 = GetSelectedLevelActors(self=Get EditorActorSubsystem())
   > Branch(Condition="integer >= integer"(A=Length(TargetArray=@n1), B=1))
   > Clear(TargetArray=Get CopyPasteTags())
   > n2 = For Each Loop(Array:Array of Actor Object References=@n1) -> (Array Element:Actor Object Reference)
   > n3 = For Each Loop(Array:Array of Names=Get Tags(self=@n2)) -> (Array Element:Name)
   > Add Unique(TargetArray=Get CopyPasteTags(), NewItem=@n3)
  ]}
 CreatePCGDataAsset {nodes [
   n1 = CreatePCGDataAsset!() -> (then, AssetName:String, PackagePath:String, Level:Soft World Reference Structure, LevelObject:Object Reference, UseLevelObject:bool=~) {is=title K2Node_FunctionEntry}
   > n2 = GetAssetTools()
   > n3 = CreateAsset(self=@n2, AssetName=@n1, PackagePath=@n1.PackagePath, AssetClass=/Script/PCG.PCGDataAsset)
   > n4 = Set OutputAsset(OutputAsset:Object Reference=@n3) -> (Output_Get:Object Reference)
   > n5 = Cast To PCGDataAsset!(execute, Object=@n4) -> (then, CastFailed, AsPCGData Asset:PCGData Asset Object Reference) {is=title K2Node_DynamicCast}
   > n6 = Set OutputPCGAsset(OutputPCGAsset:PCGData Asset Object Reference=@n5) -> (Output_Get:PCGData Asset Object Reference)
   > Set Data!(
    execute
    Data:PCGData Collection Structure=Get PCGDataCollectionOutput() -> (PCGDataCollectionOutput:PCGData Collection Structure)
    self:PCGData Asset Object Reference=@n6
   ) -> (then, Output_Get:PCGData Collection Structure)
   > Set Name!(execute, Name:String=Get Name() -> (Name:String), self:PCGData Asset Object Reference=@n5) -> (then, Output_Get:String)
   > n7 = Construct Object from Class(Class=/Script/PCGEditor.PCGLevelToAsset) -> (ReturnValue:PCGLevel To Asset Object Reference)
   > n8 = Branch(Condition=Get UseLevelObject() -> (UseLevelObject:bool))
   > n9 = SetWorldObject!(execute, self:PCGLevel To Asset Object Reference=@n7, WorldObject:Object Reference=Get LevelObject() -> (LevelObject:Object Reference)) -> (then, ReturnValue:bool)
   n10 = CreateAsset!(
    execute=[@n9.then, @n13.then]
    Exporter:PCGAsset Exporter Object Reference=@n7
    Parameters:PCGAsset Exporter Parameters Structure=Make PCGAsset Exporter Parameters!(
     bOpenSaveDialog:bool=false
     AssetName:String=Get AssetName() -> (AssetName:String)
     AssetPath:String=Get PackagePath() -> (PackagePath:String)
     bSaveOnExportEnded:bool=True
    ) -> (PCGAssetExporterParameters:PCGAsset Exporter Parameters Structure) {is=title K2Node_MakeStruct}
   ) -> (then, ReturnValue:Package Object Reference)
   > n11 = Set OutputAsset(OutputAsset=@n10)
   > Return Node!(execute, Return Value:Object Reference=@n11, Success:bool=true)
   n12 = Get Level() -> (Level:Soft World Reference Structure)
   n13 = SetWorld!(
    execute=@n8.else
    self:PCGLevel To Asset Object Reference=@n7
    World:World Object Reference=Resolve Soft Reference!(Input:World Soft Object Reference=Break Soft World Reference!(SoftWorldReference:"Soft World Reference Structure (by ref)"=@n12) -> (WorldAsset:World Soft Object Reference) {is=title K2Node_BreakStruct}) -> (Output:World Object Reference)
   ) -> (then)
   Return Node(execute=@n5.CastFailed, Success=false)
  ]}
 EventGraph {nodes [
   "On Clicked (Btn_ReplaceActorClass)"() {is=title K2Node_ComponentBoundEvent}
   > ReplaceAllSourceAndTargetActorClasses!(execute, self, ActorsToReplace:Array of Actor Object References, LevelReference:Object Reference) -> (then)
   "On Clicked (Btn_ReplaceSelectedActors)"() {is=title K2Node_ComponentBoundEvent}
   > ReplaceSelectedActors!(execute, self, OnlyCopyPredefinedTags:bool=false) -> (then)
   n1 = Event Construct()
   n2 = Get SinglePropertyView_TargetActor() -> (SinglePropertyView_TargetActor:Single Property View Object Reference)
   SetObject(execute=[@n1.then, @n5.then], self=@n2, NewObject=Self-Reference())
   > SetPropertyName!(execute, self:Single Property View Object Reference=@n2, NewPropertyName:Name=TargetActorClass) -> (then)
   n3 = Get SinglePropertyView_SourceActor() -> (SinglePropertyView_SourceActor:Single Property View Object Reference)
   > SetObject(self=@n3, NewObject=Self-Reference())
   > SetPropertyName(self=@n3, NewPropertyName=SourceActorClass)
   n4 = Get SinglePropertyView_ActorRemapDataTable() -> (SinglePropertyView_ActorRemapDataTable:Single Property View Object Reference)
   > SetObject(self=@n4, NewObject=Self-Reference())
   > SetPropertyName(self=@n4, NewPropertyName=ActorRemappingDatatable)
   n5 = Event PreConstruct()
  ]}
 PasteCopiedTagsTo SelectedActors {nodes [
   PasteCopiedTagsTo SelectedActors() {is=title K2Node_FunctionEntry}
   > n1 = GetSelectedLevelActors(self=Get EditorActorSubsystem())
   > Branch(Condition="integer >= integer"(A=Length(TargetArray=@n1), B=1))
   > n2 = For Each Loop(Array:Array of Actor Object References=@n1) -> (Array Element:Actor Object Reference)
   Return Node!(execute=@n2.Completed)
   > TransactObject(Object=@n2)
   > Clear(TargetArray=Get TempTagsArray() -> (TempTagsArray:Array of Names))
   > Set TempTagsArray(TempTagsArray:Array of Names=Get Tags(self=@n2)) -> (Output_Get:Array of Names)
   > n3 = For Each Loop(Array:Array of Names=Get CopyPasteTags()) -> (Array Element:Name)
   Set Tags!(execute=@n3.Completed, Tags:Array of Names=Get TempTagsArray(), self:Actor Object Reference=@n2) -> (then, Output_Get:Array of Names)
   Add Unique(execute=@n3.LoopBody, TargetArray=Get TempTagsArray(), NewItem=@n3)
  ]}
 PasteCopiedTagsToActor {nodes [
   n1 = PasteCopiedTagsToActor!() -> (then, Actor Reference:Actor Object Reference) {is=title K2Node_FunctionEntry}
   > TransactObject(Object=@n1)
   > Clear(TargetArray=Get TempTagsArray())
   > Set TempTagsArray(TempTagsArray=Get Tags(self=@n1))
   > n2 = For Each Loop(Array=Get CopyPasteTags())
   Set Tags(execute=@n2.Completed, Tags=Get TempTagsArray(), self=@n1)
   Add Unique(execute=@n2.LoopBody, TargetArray=Get TempTagsArray(), NewItem=@n2)
  ]}
 ReplaceAllSourceAndTargetActorClasses {nodes [
   ReplaceAllSourceAndTargetActorClasses!() -> (then, ActorsToReplace:Array of Actor Object References, LevelReference:Object Reference) {is=title K2Node_FunctionEntry}
   > n1 = GetSelectedAssetsOfClass(AssetClass=/Script/Engine.World)
   > n2 = Branch(Condition="integer > integer"(A=Length(TargetArray:Array of Object References=@n1)))
   > SearchReplaceActorsInLevelReferences!(execute, self, LevelReferences:Array of Object References=@n1, Export PCG Asset:bool=true) -> (then)
   PrintString(execute=@n2.else, InString=No Selected Levels in Content Browser, bPrintToScreen=false, bPrintToLog=Get DebugOutput() -> (DebugOutput:bool))
   > n3 = GetSelectedLevelActors(self=Get EditorActorSubsystem())
   > Branch(Condition="integer > integer"(A=Length(TargetArray:Array of Actor Object References=@n3)))
   > n4 = Filter by Class!(
    execute
    TargetArray:Array of Object References=@n3
    ObjectClass:Object Class Reference=/Script/Engine.LevelInstance
    FilterType:EEditorScriptingFilterType Enum
   ) -> (then, ReturnValue:Array of Level Instance Object References)
   > n5 = Branch(Condition="integer > integer"(A=Length(TargetArray:Array of Level Instance Object References=@n4)))
   > SearchReplaceActorsInLevelReferences(LevelReferences=@n4, Export PCG Asset=true)
   PrintString(execute=@n5.else, InString=No Selected Level Instances in Current Editor Level, bPrintToScreen=false, bPrintToLog=Get DebugOutput())
  ]}
 ReplaceSelectedActors {nodes [
   ReplaceSelectedActors!() -> (then, OnlyCopyPredefinedTags:bool=~) {is=title K2Node_FunctionEntry}
   > n1 = GetSelectedLevelActors(self=Get EditorActorSubsystem())
   > Branch(Condition="integer >= integer"(A=Length(TargetArray:Array of Actor Object References=@n1), B=1))
   > n2 = For Each Loop(Array:Array of Actor Object References=@n1) -> (Array Element:Actor Object Reference)
   > SetSelectedLevelActors(self=Get EditorActorSubsystem(), ActorsToSelect=Make Array("[0]":Actor Object Reference=@n2) -> (Array:Array of Actor Object References))
   > CopySelectedActorTags!(execute, self, UsePredefinedTags:bool=Get OnlyCopyPredefinedTags() -> (OnlyCopyPredefinedTags:bool)) -> (then) {is=title K2Node_CallFunction}
   n3 = Get TargetActorClass() -> (TargetActorClass:Object Reference)
   n4 = Get Object Path String(Object=@n3)
   > PrintString(
    InString=Append!(A="Actor Class: ", B=GetDisplayName(Object=@n3), C:String="\r\nAsset Path:", D:String=@n4) -> (ReturnValue)
    bPrintToScreen=false
    bPrintToLog=Get DebugOutput()
   )
   > ReplaceSelectedActors!(execute, InAssetPath:String=@n4) -> (then) {is=title K2Node_CallFunction}
   > PasteCopiedTagsTo SelectedActors() {is=title K2Node_CallFunction}
  ]}
 SaveLevelToPCGDataAsset {nodes [
   n1 = SaveLevelToPCGDataAsset!() -> (then, LevelReference:Object Reference, UseLevelReference:bool=~) {is=title K2Node_FunctionEntry}
   > n2 = Branch(Condition=@n1.UseLevelReference)
   > n3 = Set Level(Level:Object Reference=Get LevelReference() -> (LevelReference:Object Reference)) -> (Output_Get:Object Reference)
   n4 = Get Level() -> (Level:Object Reference)
   Set PackagePath(
    execute=[@n16.then, @n3.then]
    PackagePath:String=Split!(InPath:String=Get Object Path String(Object=@n4)) -> (PathPart:String, FilenamePart:String, ExtensionPart:String)
   ) -> (Output_Get:String)
   > Set AssetName(AssetName:String=Append!(A=GetDisplayName(Object=@n4), B=_PCG) -> (ReturnValue)) -> (Output_Get:String)
   n5 = Get AssetName()
   > Set AssetPath(AssetPath:String=Append!(A=Get PackagePath(), B=/, C:String=@n5, D:String=., E:String=@n5) -> (ReturnValue)) -> (Output_Get:String)
   > n6 = DoesAssetExist(AssetPath=Get AssetPath() -> (AssetPath:String))
   > n7 = Branch(Condition=@n6)
   > PrintString(InString="Preview Asset Exists ->Loading Preview Asset", bPrintToScreen=false, bPrintToLog=Get DebugOutput())
   > n8 = LoadAsset(AssetPath=Get AssetPath())
   > n9 = Cast To PCGDataAsset(Object=@n8)
   n10 = MakeSoftObjectPath!(PathString:String=Get Object Path String(Object=Get Level())) -> (ReturnValue:Soft Object Path Structure)
   > Set ObjectPath!(execute, ObjectPath:Soft Object Path Structure=@n10, self:PCGData Asset Object Reference=@n9) -> (then, Output_Get:Soft Object Path Structure)
   n11 = Make PCGAsset Exporter Parameters(bOpenSaveDialog=false, AssetName=Get AssetName(), AssetPath=Get PackagePath(), bSaveOnExportEnded=True)
   > UpdateAssets!(
    execute
    PCGAssets:Array of Asset Data Structures=Make Array("[0]":Asset Data Structure=CreateAssetData(InAsset=@n9, bAllowBlueprintClass=false)) -> (Array:Array of Asset Data Structures)
    Parameters:PCGAsset Exporter Parameters Structure=@n11
   ) -> (then)
   > Return Node!(execute, Success:bool=true)
   PrintString(execute=[@n15.CastFailed, @n9.CastFailed], InString=Failed to Cast to PCGDataAsset, bPrintToScreen=false, bPrintToLog=Get DebugOutput())
   > Return Node(Success=false)
   n12 = PrintString(execute=@n7.else, InString="Preview Asset Not Found -> Creating New Preview Asset", bPrintToScreen=false, bPrintToLog=Get DebugOutput())
   n13 = GetEditorWorld(execute=@n2.else, self=Get UnrealEditorSubsystem())
   n14 = CreatePCGDataAsset!(
    execute=@n12.then
    self
    AssetName:String=Get AssetName()
    PackagePath:String=Get PackagePath()
    Level:Soft World Reference Structure=Make Soft World Reference!(WorldAsset:World Soft Object Reference=Make Soft Reference!(Input:World Object Reference=@n13) -> (Output:World Soft Object Reference)) -> (SoftWorldReference:Soft World Reference Structure) {is=title K2Node_MakeStruct}
    LevelObject:Object Reference=Get LevelReference()
    UseLevelObject:bool=Get UseLevelReference() -> (UseLevelReference:bool)
   ) -> (then, Return Value:Object Reference, Success:bool) {is=title K2Node_CallFunction}
   > n15 = Cast To PCGDataAsset(Object=@n14)
   > Set ObjectPath(ObjectPath=@n10, self=@n15)
   > UpdateAssets(PCGAssets=Make Array("[0]"=CreateAssetData(InAsset=@n15, bAllowBlueprintClass=false)), Parameters=@n11)
   > Return Node(Success=true)
   n16 = Set Level(execute=@n13.then, Level=@n13)
  ]}
 SearchReplaceActorsInLevelReferences {nodes [
   n1 = Get SourceActorClass() -> (SourceActorClass:Object Reference)
   n2 = Split!(SourceString:String=Get Object Path String(Object=@n1), InStr:String=., SearchCase:ESearchCase Enum, SearchDir:ESearchDir Enum=FromStart) -> (LeftS:String, RightS:String, ReturnValue:bool)
   n3 = LoadAsset(AssetPath=@n2)
   > n4 = Cast To Actor!(execute, Object=@n3) -> (then, CastFailed, AsActor:Actor Object Reference) {is=title K2Node_DynamicCast}
   n5 = LoadBlueprintClass!(execute=@n41.else, AssetPath:String=@n2) -> (then, ReturnValue:Object Class Reference)
   > n6 = Cast To Actor Class!(execute:Exec, Class:Object Class Reference=@n5) -> (then:Exec, CastFailed:Exec, AsActor:Actor Class Reference) {is=title K2Node_ClassDynamicCast}
   n7 = GetAllActorsOfClassWithTag!(
    execute
    WorldContextObject:Object Reference=Get LocalCurrentLevel() -> (LocalCurrentLevel:Object Reference)
    ActorClass:Actor Class Reference=@n6
    Tag:Name=ASM
   ) -> (then, OutActors:Array of Actor Object References)
   > n8 = Set LocalCurrentLevelActors(LocalCurrentLevelActors:Array of Actor Object References=@n7) -> (Output_Get:Array of Actor Object References)
   Clear(
    execute=[@n73.then, @n8.then]
    TargetArray:Array of Actor Object References=Get LocalReplacedActors() -> (LocalReplacedActors:Array of Actor Object References)
   )
   > n9 = For Each Loop(Array=Get LocalCurrentLevelActors() -> (LocalCurrentLevelActors:Array of Actor Object References))
   n10 = For Each Loop(Exec=@n9.Completed, Array=Get LocalReplacedActors())
   Sequence!(execute=@n10.Completed) -> (then_0, then_1, then_2:Exec)
   > Clear(TargetArray=Get LocalReplacedActors())
   > Branch(Condition=Get LevelDirty() -> (LevelDirty:bool))
   > n11 = SaveLoadedAsset(AssetToSave=Get LocalCurrentLevel(), bOnlyIfIsDirty=false)
   > n12 = Branch(Condition=@n11)
   > Branch(Condition=Get Export PCG Asset() -> (Export PCG Asset:bool))
   > SaveLevelToPCGDataAsset!(execute, self, LevelReference:Object Reference=Get LocalCurrentLevel(), UseLevelReference:bool=true) -> (then, Success:bool) {is=title K2Node_CallFunction}
   PrintString(execute=@n12.else, InString=3::Failed to save the level file, bPrintToScreen=false, bPrintToLog=Get DebugOutput())
   BeginTransaction(execute=@n10.LoopBody, Context=Destroying replaced Actors)
   > Destroy Actor(self=@n10)
   > EndTransaction()
   n13 = Set LocalCurrentActor(execute=@n9.LoopBody, LocalCurrentActor:Actor Object Reference=@n9) -> (Output_Get:Actor Object Reference)
   > Set LocalParentActor(LocalParentActor:Actor Object Reference=GetAttachParentActor!(self:Actor Object Reference=@n13) -> (ReturnValue:Actor Object Reference)) -> (Output_Get:Actor Object Reference)
   n14 = Split(SourceString=Get Object Path String(Object=Get TargetActorClass()), InStr=., SearchDir=FromStart)
   > n15 = LoadBlueprintClass(AssetPath=@n14)
   > n16 = Cast To Actor Class(Class=@n15)
   > n17 = SpawnActor!(
    execute:Exec
    WorldContextObject:Object Reference=Get LocalCurrentLevel()
    Class:Actor Class Reference=@n16
    SpawnTransform:Transform=Get Actor Transform(self=Get LocalCurrentActor() -> (LocalCurrentActor:Actor Object Reference))
    CollisionHandlingOverride:ESpawnActorCollisionHandlingMethod Enum=AlwaysSpawn
    TransformScaleMethod:ESpawnActorScaleMethod Enum=MultiplyWithRoot
    Owner:Actor Object Reference
    Instigator:Pawn Object Reference
   ) -> (then:Exec, ReturnValue:Actor Object Reference) {is=title K2Node_SpawnActorFromClass}
   > Add Unique(TargetArray:Array of Actor Object References=Get LocalReplacedActors(), NewItem:"Actor Object Reference (by ref)"=Get LocalCurrentActor())
   > Set LevelDirty(LevelDirty:bool=true) -> (Output_Get:bool)
   > Set LocalCurrentSpawnedActor(LocalCurrentSpawnedActor:Actor Object Reference=@n17) -> (Output_Get:Actor Object Reference)
   > CopyActorTags!(execute, self, ActorReference:Actor Object Reference=Get LocalCurrentActor()) -> (then) {is=title K2Node_CallFunction}
   > Remove Item!(execute, TargetArray:Array of Names=Get Tags(self=Get LocalCurrentActor()), Item:"Name (by ref)"=ASM) -> (then, ReturnValue:bool)
   > PasteCopiedTagsToActor!(execute, self, Actor Reference:Actor Object Reference=Get LocalCurrentSpawnedActor() -> (LocalCurrentSpawnedActor:Actor Object Reference)) -> (then) {is=title K2Node_CallFunction}
   n18 = Get LocalParentActor() -> (LocalParentActor:Actor Object Reference)
   > n19 = Branch(Condition=IsValid(Object=@n18))
   > n20 = Attach Component To Component!(
    execute
    self:Scene Component Object Reference=Get RootComponent!(self:Actor Object Reference=Get LocalCurrentSpawnedActor()) -> (RootComponent:Scene Component Object Reference)
    Parent:Scene Component Object Reference=Get RootComponent(self=@n18)
    SocketName:Name=None
    LocationRule:EAttachmentRule Enum=KeepWorld
    RotationRule:EAttachmentRule Enum=KeepWorld
    ScaleRule:EAttachmentRule Enum=KeepWorld
    bWeldSimulatedBodies:bool=false
   ) -> (then, ReturnValue:bool)
   n21 = For Each Loop(
    Exec=[@n20.then, @n19.else]
    Array=GetAttachedActors!(self:Actor Object Reference=Get LocalCurrentActor(), bResetArray:bool=true, bRecursivelyIncludeAttachedActors:bool=true) -> (OutActors:Array of Actor Object References)
   )
   > n22 = Attach Component To Component(
    self=Get RootComponent(self=@n21)
    Parent=Get RootComponent(self=Get LocalCurrentSpawnedActor())
    SocketName=None
    LocationRule=KeepWorld
    RotationRule=KeepWorld
    ScaleRule=KeepWorld
    bWeldSimulatedBodies=false
   )
   > n23 = Branch(Condition=@n22)
   PrintString(
    execute=@n23.else
    InString=Append!(A=GetDisplayName(Object=@n21), B=" Failed to Attach to Parent") -> (ReturnValue)
    bPrintToScreen=false
    bPrintToLog=Get DebugOutput()
   )
   PrintString(
    execute=@n16.CastFailed
    InString=Append(A="1::Cast to Target Actor Class Failed for \r\n--", B=@n14)
    bPrintToScreen=false
    bPrintToLog=Get DebugOutput()
   )
   n24 = Get LocalCurrentActor()
   n25 = Get Actor Transform(self=@n24)
   n26 = BreakTransform(InTransform=@n25)
   n27 = Get TargetActorClass()
   n28 = SpawnActorFromObject!(
    execute
    ObjectToUse:Object Reference=GetAsset(InAssetData=CreateAssetData(InAsset=@n27, bAllowBlueprintClass=false))
    Location:Vector=@n26
    Rotation:Rotator=@n26.Rotation
    bTransient:bool=false
   ) -> (then, ReturnValue:Actor Object Reference)
   > n29 = Branch(Condition=IsValid(Object=@n28))
   > SetActorScale3D!(execute, self:Actor Object Reference=@n28, NewScale3D:Vector=GetActorScale3D!(self:Actor Object Reference=@n24) -> (ReturnValue:Vector)) -> (then)
   > Add Unique(TargetArray=Get LocalReplacedActors(), NewItem=Get LocalCurrentActor())
   > Set LevelDirty(LevelDirty=true)
   > n30 = Set LocalCurrentSpawnedActor(LocalCurrentSpawnedActor=@n28)
   CopyActorTags(execute=[@n66.then, @n30.then], ActorReference=Get LocalCurrentActor())
   > Remove Item(TargetArray=Get Tags(self=Get LocalCurrentActor()), Item=ASM)
   > PasteCopiedTagsToActor(Actor Reference=Get LocalCurrentSpawnedActor())
   n31 = Get LocalParentActor()
   > n32 = Branch(Condition=IsValid(Object=@n31))
   > n33 = Attach Component To Component(
    self=Get RootComponent(self=Get LocalCurrentSpawnedActor())
    Parent=Get RootComponent(self=@n31)
    SocketName=None
    LocationRule=KeepWorld
    RotationRule=KeepWorld
    ScaleRule=KeepWorld
    bWeldSimulatedBodies=false
   )
   n34 = For Each Loop(Exec=[@n33.then, @n32.else], Array=GetAttachedActors(self=Get LocalCurrentActor(), bResetArray=true, bRecursivelyIncludeAttachedActors=true))
   > n35 = Attach Component To Component(
    self=Get RootComponent(self=@n34)
    Parent=Get RootComponent(self=Get LocalCurrentSpawnedActor())
    SocketName=None
    LocationRule=KeepWorld
    RotationRule=KeepWorld
    ScaleRule=KeepWorld
    bWeldSimulatedBodies=false
   )
   > n36 = Branch(Condition=@n35)
   PrintString(
    execute=@n36.else
    InString=Append(A=GetDisplayName(Object=@n34), B=" Failed to Attach to Parent")
    bPrintToScreen=false
    bPrintToLog=Get DebugOutput()
   )
   PrintString(execute=[@n29.else, @n65.else], InString=3:: Failed to Spawn Replacement Actor, bPrintToScreen=false, bPrintToLog=Get DebugOutput())
   SearchReplaceActorsInLevelReferences!() -> (then, LevelReferences:Array of Object References, Export PCG Asset:bool=~) {is=title K2Node_FunctionEntry}
   > Sequence!(execute) -> (then_0, then_1)
   > n37 = For Each Loop(Array:Array of Object References=Get LevelReferences() -> (LevelReferences:Array of Object References)) -> (Array Element:Object Reference)
   Set LocalCurrentActor(execute=@n37.Completed)
   > Set LocalCurrentSpawnedActor()
   > Set LocalCurrentLevel(LocalCurrentLevel:Object Reference) -> (Output_Get:Object Reference)
   > n38 = For Each Loop(Array=Get LevelReferences())
   CollectGarbage!(execute=@n38.Completed) -> (then)
   > PrintString(InString=6:2: Exiting, bPrintToScreen=false, bPrintToLog=Get DebugOutput())
   n39 = UnloadPackages!(
    execute=@n38.LoopBody
    PackagesToUnload:Array of Package Object References=Make Array("[0]":Package Object Reference=GetPackageForObject!(Object:Object Reference=@n38) -> (ReturnValue:Package Object Reference)) -> (Array:Array of Package Object References)
   ) -> (then, bOutAnyPackagesUnloaded:bool, OutErrorMessage:Text)
   > n40 = Branch(Condition=@n39)
   > PrintString(InString=5::Level Unloaded from Memory, bPrintToScreen=false, bPrintToLog=Get DebugOutput())
   PrintText(execute=@n40.else, InText=@n39.OutErrorMessage, bPrintToScreen=false, bPrintToLog=Get DebugOutput())
   Set LocalCurrentLevel(execute=@n37.LoopBody, LocalCurrentLevel=GetAsset(InAssetData=CreateAssetData(InAsset=@n37, bAllowBlueprintClass=false)))
   > Set LevelDirty(LevelDirty=false)
   > Clear(TargetArray=Get LocalReplacedActors())
   > n41 = Branch(Condition=IsValid(Object=Get ActorRemappingDatatable() -> (ActorRemappingDatatable:Object Reference)))
   > Sequence!(execute) -> (then_0, then_1, then_2:Exec, then_3:Exec)
   > Clear(TargetArray=Get LocalReplacedActors())
   > n42 = Cast To DataTable!(execute, Object=Get ActorRemappingDatatable()) -> (then, CastFailed, AsData Table:Data Table Object Reference) {is=title K2Node_DynamicCast}
   > n43 = GetDataTableRowNames(Table=@n42)
   > n44 = For Each Loop(Array:Array of Names=@n43) -> (Array Element:Name)
   Set LocalCurrentActor(execute=@n44.Completed)
   > BeginTransaction(Context=Destroying replaced Actors)
   > n45 = For Each Loop(Array:Array of Actor Object References=Get LocalReplacedActors()) -> (Array Element:Actor Object Reference)
   Clear(execute=@n45.Completed, TargetArray=Get LocalReplacedActors())
   > EndTransaction()
   > Sequence!(execute) -> (then_0, then_1, then_2)
   > Branch(Condition=Get LevelDirty())
   > n46 = SaveLoadedAsset(AssetToSave=Get LocalCurrentLevel(), bOnlyIfIsDirty=false)
   > n47 = Branch(Condition=@n46)
   > Branch(Condition=Get Export PCG Asset())
   > SaveLevelToPCGDataAsset(LevelReference=Get LocalCurrentLevel(), UseLevelReference=true)
   PrintString(execute=@n47.else, InString=3::Failed to save the level file, bPrintToScreen=false, bPrintToLog=Get DebugOutput())
   Destroy Actor(execute=@n45.LoopBody, self=@n45)
   Set SourceActorClass(execute=@n44.LoopBody, SourceActorClass:Object Reference) -> (Output_Get:Object Reference)
   > Set TargetActorClass(TargetActorClass:Object Reference) -> (Output_Get:Object Reference)
   > Clear(TargetArray=Get LocalCurrentLevelActors())
   > n48 = Get Data Table Row!(execute:Exec, DataTable:Data Table Object Reference=@n42, RowName:Name=@n44) -> (then:Exec, RowNotFound:Exec, ReturnValue:S Actor Tagger Actor Replace Structure)
   > n49 = Sequence!(execute) -> (then_0, then_1)
   n50 = Break S Actor Tagger Actor Replace!(S_ActorTaggerActorReplace:"S Actor Tagger Actor Replace Structure (by ref)"=@n48) -> (
    SourceActorClass_17_#1:Soft Object Reference
    SourceIsStaticMeshAsset_21_#2:bool=~
    TargetActorClass_18_#3:Soft Object Reference
   ) {is=title K2Node_BreakStruct}
   n51 = "To String (SoftObjectReference)"!(SoftObjectReference:"Soft Object Reference (by ref)"=@n50) -> (ReturnValue:String)
   > n52 = LoadAsset(AssetPath=@n51)
   > Set SourceActorClass(SourceActorClass=@n52)
   > n53 = LoadAsset(AssetPath="To String (SoftObjectReference)"(SoftObjectReference=@n50.TargetActorClass_18_#3))
   > Set TargetActorClass(TargetActorClass=@n53)
   > Branch(Condition=@n50.SourceIsStaticMeshAsset_21_#2)
   > PrintString(InString=Skipping Staticmesh Source Actor, bPrintToScreen=false, bPrintToLog=Get DebugOutput())
   > Set StaticMeshSourceActor(StaticMeshSourceActor:bool=true) -> (Output_Get:bool)
   n54 = Branch(execute=@n49.then_1, Condition=Get StaticMeshSourceActor() -> (StaticMeshSourceActor:bool))
   > n55 = GetAllActorsOfClass!(execute, WorldContextObject:Object Reference=Get LocalCurrentLevel(), ActorClass:Actor Class Reference=&StaticMeshActor) -> (then, OutActors:Array of Static Mesh Actor Object References)
   > n56 = For Each Loop(Array:Array of Static Mesh Actor Object References=@n55) -> (Array Element:Static Mesh Actor Object Reference)
   n57 = Get StaticMesh(self=GetComponentByClass!(self:Actor Object Reference=@n56, ComponentClass:Actor Component Class Reference=&StaticMeshComponent) -> (ReturnValue:Static Mesh Component Object Reference))
   > Branch(Condition="Equal Exactly (String)"(A=@n51, B=Get Object Path String(Object=@n57)))
   > Add(TargetArray:Array of Actor Object References=Get LocalCurrentLevelActors(), NewItem:"Actor Object Reference (by ref)"=@n56) {is=title K2Node_CallArrayFunction}
   n58 = For Each Loop(Exec=[@n71.then, @n56.Completed], Array:Array of Actor Object References=Get LocalCurrentLevelActors()) -> (Array Element:Actor Object Reference)
   Set StaticMeshSourceActor(execute=@n58.Completed, StaticMeshSourceActor=false)
   n59 = Set LocalCurrentActor(execute=@n58.LoopBody, LocalCurrentActor=@n58)
   > Set LocalParentActor(LocalParentActor=GetAttachParentActor(self=@n59))
   > n60 = Branch(Condition=Get StaticMeshSourceActor())
   n61 = Split(SourceString=Get Object Path String(Object=@n27), InStr=., SearchDir=FromStart)
   n62 = LoadBlueprintClass(execute=[@n60.else, @n60.then], AssetPath=@n61)
   > n63 = Cast To Actor Class(Class=@n62)
   > n64 = SpawnActor(
    WorldContextObject=Get LocalCurrentLevel()
    Class=@n63
    SpawnTransform=@n25
    CollisionHandlingOverride=AlwaysSpawn
    TransformScaleMethod=MultiplyWithRoot
   )
   > n65 = Branch(Condition=IsValid(Object=@n64))
   > Add Unique(TargetArray=Get LocalReplacedActors(), NewItem=Get LocalCurrentActor())
   > Set LevelDirty(LevelDirty=true)
   > n66 = Set LocalCurrentSpawnedActor(LocalCurrentSpawnedActor=@n64)
   PrintString(
    execute=@n63.CastFailed
    InString=Append(A="1::Cast to Target Actor Class Failed\r\n-- ", B=@n61)
    bPrintToScreen=false
    bPrintToLog=Get DebugOutput()
   )
   n67 = Get SourceActorClass()
   n68 = LoadBlueprintClass(execute=@n54.else, AssetPath=Split(SourceString=Get Object Path String(Object=@n67), InStr=., SearchDir=FromStart))
   > n69 = Cast To Actor Class(Class=@n68)
   > n70 = GetAllActorsOfClass(WorldContextObject=Get LocalCurrentLevel(), ActorClass=@n69) -> (OutActors:Array of Actor Object References)
   > n71 = Set LocalCurrentLevelActors(LocalCurrentLevelActors=@n70)
   PrintString(
    execute=@n69.CastFailed
    InString=4:: Failed to Cast loaded BP Source Actor Class to Actor Class
    bPrintToScreen=false
    bPrintToLog=Get DebugOutput()
   )
   PrintString(
    execute=@n42.CastFailed
    InString=No DataTable specified or the datatable specified is not of the correct row struct type.
    bPrintToScreen=false
    bPrintToLog=Get DebugOutput()
   )
   PrintString(execute=@n6.CastFailed, InString=0:: Cast to Actor Failed, bPrintToScreen=false, bPrintToLog=Get DebugOutput())
   Get Class(Object=@n4)
   Get Class(Object=@n1) -> (ReturnValue:Object Class Reference)
   n72 = GetAllActorsOfClass(execute=@n74.else, WorldContextObject=Get LocalCurrentLevel(), ActorClass=@n6)
   > n73 = Set LocalCurrentLevelActors(LocalCurrentLevelActors=@n72)
   n74 = Branch(execute=@n6.then, Condition=false)
  ]}
]
variables [
 ActorRemappingDatatable {type=Object, default=null}
 CopyPasteTags {type="array<string>", default []}
 DebugOutput {type=boolean}
 SourceActorClass {type=Object, default=null}
 TargetActorClass {type=Object, default=null}
 TempTagsArray {type="array<string>", default []}
]
widgets {
 info {rootWidgetClass=&Border, widgetCount=16}
 widgets [
  Border_244 {widgetClassPath=&Border, widget=Border_244, properties {verticalAlignment=VAlign_Top, brushColor=<rgba 0.0219 0.0219 0.0219 1>}}
  VerticalBox_401 {
   widgetClassPath=&VerticalBox
   parent=Border_244
   widget=VerticalBox_401
   slot {class=BorderSlot, properties {verticalAlignment=VAlign_Top}}
  }
  UniformGridPanel_74 {widgetClassPath=/Script/UMG.UniformGridPanel, parent=VerticalBox_401, widget=UniformGridPanel_74, slot {class=VerticalBoxSlot}}
  Spacer_5 {
   widgetClassPath=&Spacer
   parent=VerticalBox_401
   widget=Spacer_5
   properties {size=<1 8>}
   slot {class=VerticalBoxSlot, properties {horizontalAlignment=HAlign_Left}}
  }
  Spacer_6 {
   widgetClassPath=&Spacer
   parent=VerticalBox_401
   widget=Spacer_6
   properties {size=<1 8>}
   slot {class=VerticalBoxSlot, properties {horizontalAlignment=HAlign_Left}}
  }
  Border_9 {
   widgetClassPath=&Border
   parent=VerticalBox_401
   widget=Border_9
   properties {brushColor=<rgba 0.0135 0.0135 0.0135 1>, padding=<ltrb 0 0 0 0>}
   slot {class=VerticalBoxSlot}
  }
  SinglePropertyView_SourceActor {
   widgetClassPath=&SinglePropertyView
   parent=Border_9
   widget=SinglePropertyView_SourceActor
   properties {bAutoLoadAsset=false, bIsVolatile=true, propertyName=ActorClass, toolTipText=Class to be substituted}
   slot {class=BorderSlot, properties {padding=<ltrb 0 0 0 0>}}
  }
  Border_2 {
   widgetClassPath=&Border
   parent=VerticalBox_401
   widget=Border_2
   properties {brushColor=<rgba 0.0135 0.0135 0.0135 1>, padding=<ltrb 0 0 0 0>}
   slot {class=VerticalBoxSlot}
  }
  SinglePropertyView_TargetActor {
   widgetClassPath=&SinglePropertyView
   parent=Border_2
   widget=SinglePropertyView_TargetActor
   properties {bAutoLoadAsset=false, bIsVolatile=true, propertyName=ActorClass, toolTipText=Substitute Class}
   slot {class=BorderSlot, properties {padding=<ltrb 0 0 0 0>}}
  }
  SinglePropertyView_ActorRemapDataTable {
   widgetClassPath=&SinglePropertyView
   parent=VerticalBox_401
   widget=SinglePropertyView_ActorRemapDataTable
   properties {bAutoLoadAsset=false, bIsVolatile=true, propertyName=ActorClass, toolTipText=DataTable Reference to use for Actor Remapping}
   slot {class=VerticalBoxSlot}
  }
  Spacer_7 {
   widgetClassPath=&Spacer
   parent=VerticalBox_401
   widget=Spacer_7
   properties {size=<1 8>}
   slot {class=VerticalBoxSlot, properties {horizontalAlignment=HAlign_Center, verticalAlignment=VAlign_Center}}
  }
  UniformGridPanel_279 {
   widgetClassPath=&UniformGridPanel
   parent=VerticalBox_401
   widget=UniformGridPanel_279
   slot {class=VerticalBoxSlot, properties {verticalAlignment=VAlign_Top}}
  }
  Btn_ReplaceActorClass {
   widgetClassPath=&Button
   parent=UniformGridPanel_279
   widget=Btn_ReplaceActorClass
   properties {
    toolTipText=Replace the Source Actor class with the Target Actor Class. All actors of the same class will be replaced while maintaining hierarchy and tags copied from the replaced actor. You can also use a datatable to remap mutltiple actors in the content browser selected levels. Selected level instances in a level can be processed also.
    widgetStyle {
     clickedSlateSound {resourceObject=null}
     disabled {
      bIsDynamicallyLoaded=false
      drawAs=RoundedBox
      imageType=NoImage
      mirroring=NoMirror
      resourceName=None
      resourceObject=null
      tiling=NoTile
      imageSize=<0 0>
      margin=<ltrb 0 0 0 0>
      outlineSettings {
       bUseBrushTransparency=false
       roundingType=FixedRadius
       width=1
       color {colorUseRule=UseColor_ColorTable, specifiedColor=<rgba 0 0 0 0>}
       cornerRadii=<4 4 4 4>
      }
      tintColor {colorUseRule=UseColor_Specified, specifiedColor=<rgba 0.0395 0.0395 0.0395 1>}
      uVRegion {bIsValid=false, max=<0 0>, min=<0 0>}
     }
     disabledForeground {colorUseRule=UseColor_ColorTable, specifiedColor=<rgba 0 0 0 0>}
     hovered {
      bIsDynamicallyLoaded=false
      drawAs=RoundedBox
      imageType=NoImage
      mirroring=NoMirror
      resourceName=None
      resourceObject=null
      tiling=NoTile
      imageSize=<32 32>
      margin=<ltrb 0 0 0 0>
      outlineSettings {
       bUseBrushTransparency=false
       roundingType=FixedRadius
       width=1
       color {colorUseRule=UseColor_ColorTable, specifiedColor=<rgba 0 0 0 0>}
       cornerRadii=<4 4 4 4>
      }
      tintColor {colorUseRule=UseColor_Specified, specifiedColor=<rgba 0.0953 0.0953 0.0953 1>}
      uVRegion {bIsValid=false, max=<0 0>, min=<0 0>}
     }
     hoveredForeground {colorUseRule=UseColor_ColorTable, specifiedColor=<rgba 0 0 0 0>}
     hoveredSlateSound {resourceObject=null}
     normal {
      bIsDynamicallyLoaded=false
      drawAs=RoundedBox
      imageType=NoImage
      mirroring=NoMirror
      resourceName=None
      resourceObject=null
      tiling=NoTile
      imageSize=<32 32>
      margin=<ltrb 0 0 0 0>
      outlineSettings {
       bUseBrushTransparency=false
       roundingType=FixedRadius
       width=1
       color {colorUseRule=UseColor_ColorTable, specifiedColor=<rgba 0 0 0 0>}
       cornerRadii=<4 4 4 4>
      }
      tintColor {colorUseRule=UseColor_Specified, specifiedColor=<rgba 0.0395 0.0395 0.0395 1>}
      uVRegion {bIsValid=false, max=<0 0>, min=<0 0>}
     }
     normalForeground {colorUseRule=UseColor_ColorTable, specifiedColor=<rgba 0 0 0 0>}
     normalPadding=<ltrb 12 1.5 12 1.5>
     pressed {
      bIsDynamicallyLoaded=false
      drawAs=RoundedBox
      imageType=NoImage
      mirroring=NoMirror
      resourceName=None
      resourceObject=null
      tiling=NoTile
      imageSize=<32 32>
      margin=<ltrb 0 0 0 0>
      outlineSettings {
       bUseBrushTransparency=false
       roundingType=FixedRadius
       width=1
       color {colorUseRule=UseColor_ColorTable, specifiedColor=<rgba 0 0 0 0>}
       cornerRadii=<4 4 4 4>
      }
      tintColor {colorUseRule=UseColor_Specified, specifiedColor=<rgba 0.0284 0.0284 0.0284 1>}
      uVRegion {bIsValid=false, max=<0 0>, min=<0 0>}
     }
     pressedForeground {colorUseRule=UseColor_ColorTable, specifiedColor=<rgba 0 0 0 0>}
     pressedPadding=<ltrb 12 2.5 12 0.5>
     pressedSlateSound {resourceObject=null}
    }
   }
   slot {class=UniformGridSlot, properties {horizontalAlignment=HAlign_Fill, verticalAlignment=VAlign_Center}}
  }
  TextBlock_19 {
   widgetClassPath=&TextBlock
   parent=Btn_ReplaceActorClass
   widget=TextBlock_19
   properties {
    justification=Center
    text=Replace Source Actor Class with Target Actor Class
    font {
     bForceMonospaced=false
     bMaterialIsStencil=false
     fontMaterial=null
     fontObject=&Roboto
     letterSpacing=0
     monospacedWidth=1
     size=7.5
     skewAmount=0
     typefaceFontName=Regular
     outlineSettings {
      bApplyOutlineToDropShadows=false
      bMiteredCorners=false
      bSeparateFillAlpha=false
      outlineMaterial=null
      outlineSize=0
      outlineColor=<rgba 0 0 0 1>
     }
    }
   }
   slot {class=ButtonSlot}
  }
  Btn_ReplaceSelectedActors {
   widgetClassPath=&Button
   parent=UniformGridPanel_279
   widget=Btn_ReplaceSelectedActors
   properties {
    toolTipText=Replace the selected Actor with the Replace Actor Reference. Tags will be copied across from each of the replaced actors.
    widgetStyle {
     clickedSlateSound {resourceObject=null}
     disabled {
      bIsDynamicallyLoaded=false
      drawAs=RoundedBox
      imageType=NoImage
      mirroring=NoMirror
      resourceName=None
      resourceObject=null
      tiling=NoTile
      imageSize=<0 0>
      margin=<ltrb 0 0 0 0>
      outlineSettings {
       bUseBrushTransparency=false
       roundingType=FixedRadius
       width=1
       color {colorUseRule=UseColor_ColorTable, specifiedColor=<rgba 0 0 0 0>}
       cornerRadii=<4 4 4 4>
      }
      tintColor {colorUseRule=UseColor_Specified, specifiedColor=<rgba 0.0395 0.0395 0.0395 1>}
      uVRegion {bIsValid=false, max=<0 0>, min=<0 0>}
     }
     disabledForeground {colorUseRule=UseColor_ColorTable, specifiedColor=<rgba 0 0 0 0>}
     hovered {
      bIsDynamicallyLoaded=false
      drawAs=RoundedBox
      imageType=NoImage
      mirroring=NoMirror
      resourceName=None
      resourceObject=null
      tiling=NoTile
      imageSize=<32 32>
      margin=<ltrb 0 0 0 0>
      outlineSettings {
       bUseBrushTransparency=false
       roundingType=FixedRadius
       width=1
       color {colorUseRule=UseColor_ColorTable, specifiedColor=<rgba 0 0 0 0>}
       cornerRadii=<4 4 4 4>
      }
      tintColor {colorUseRule=UseColor_Specified, specifiedColor=<rgba 0.0953 0.0953 0.0953 1>}
      uVRegion {bIsValid=false, max=<0 0>, min=<0 0>}
     }
     hoveredForeground {colorUseRule=UseColor_ColorTable, specifiedColor=<rgba 0 0 0 0>}
     hoveredSlateSound {resourceObject=null}
     normal {
      bIsDynamicallyLoaded=false
      drawAs=RoundedBox
      imageType=NoImage
      mirroring=NoMirror
      resourceName=None
      resourceObject=null
      tiling=NoTile
      imageSize=<32 32>
      margin=<ltrb 0 0 0 0>
      outlineSettings {
       bUseBrushTransparency=false
       roundingType=FixedRadius
       width=1
       color {colorUseRule=UseColor_ColorTable, specifiedColor=<rgba 0 0 0 0>}
       cornerRadii=<4 4 4 4>
      }
      tintColor {colorUseRule=UseColor_Specified, specifiedColor=<rgba 0.0395 0.0395 0.0395 1>}
      uVRegion {bIsValid=false, max=<0 0>, min=<0 0>}
     }
     normalForeground {colorUseRule=UseColor_ColorTable, specifiedColor=<rgba 0 0 0 0>}
     normalPadding=<ltrb 12 1.5 12 1.5>
     pressed {
      bIsDynamicallyLoaded=false
      drawAs=RoundedBox
      imageType=NoImage
      mirroring=NoMirror
      resourceName=None
      resourceObject=null
      tiling=NoTile
      imageSize=<32 32>
      margin=<ltrb 0 0 0 0>
      outlineSettings {
       bUseBrushTransparency=false
       roundingType=FixedRadius
       width=1
       color {colorUseRule=UseColor_ColorTable, specifiedColor=<rgba 0 0 0 0>}
       cornerRadii=<4 4 4 4>
      }
      tintColor {colorUseRule=UseColor_Specified, specifiedColor=<rgba 0.0284 0.0284 0.0284 1>}
      uVRegion {bIsValid=false, max=<0 0>, min=<0 0>}
     }
     pressedForeground {colorUseRule=UseColor_ColorTable, specifiedColor=<rgba 0 0 0 0>}
     pressedPadding=<ltrb 12 2.5 12 0.5>
     pressedSlateSound {resourceObject=null}
    }
   }
   slot {class=UniformGridSlot, properties {column=1, horizontalAlignment=HAlign_Fill, verticalAlignment=VAlign_Center}}
  }
  TextBlock_31 {
   widgetClassPath=&TextBlock
   parent=Btn_ReplaceSelectedActors
   widget=TextBlock_31
   properties {
    justification=Center
    text=Replace Selected Actors with Target Actor Class
    font {
     bForceMonospaced=false
     bMaterialIsStencil=false
     fontMaterial=null
     fontObject=&Roboto
     letterSpacing=0
     monospacedWidth=1
     size=7.5
     skewAmount=0
     typefaceFontName=Regular
     outlineSettings {
      bApplyOutlineToDropShadows=false
      bMiteredCorners=false
      bSeparateFillAlpha=false
      outlineMaterial=null
      outlineSize=0
      outlineColor=<rgba 0 0 0 1>
     }
    }
   }
   slot {class=ButtonSlot}
  }
 ]
}
