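# Editor utility widget for the PCG ActorTagger: a labeled, automated tag input that tags an actor with its class asset path.
# FormatTagValue calls GetActorClassDefaultObject, which uses FindPartialTag to strip tags matching the given tag's pre-colon text.
# It then appends the class path to the tag text, adds that uniquely to the actor's Tags and returns it.
# TransactObject calls record the actor edits for transactions, and debug prints remain throughout.
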EditorUtilityWidgetBlueprint /PCG/Utilities/Assemblies/ActorTagger/Widgets/Widget_InputClassDefaultObject
graphs [
 CleanUp {nodes [CleanUp() {is=title K2Node_FunctionEntry}]}
 EventGraph {nodes [
   Event PreConstruct()
   Event Construct()
   Event Tick()
  ]}
 FindPartialTag {nodes [
   n1 = FindPartialTag!() -> (then, TagsToSearch:Array of Names, TagText:String, RemoveFoundTags:bool=~, AllowTagsWithoutColons:bool=~) {is=title K2Node_FunctionEntry}
   > Set LocalTags(LocalTags:Array of Names=@n1) -> (Output_Get:Array of Names)
   n2 = Length(TargetArray:Array of Names=@n1)
   > PrintString(InString=Append(A=Number of Tags Passed in ::, B="To String (Integer)"(InInt=@n2)))
   > Set LocalPartialTagFound(LocalPartialTagFound:bool=false) -> (Output_Get:bool)
   > n3 = For Each Loop(Array:Array of Names=Get LocalTags() -> (LocalTags:Array of Names)) -> (Array Element:Name)
   Remove Item!(execute=@n3.Completed, TargetArray:Array of Names=Get LocalTags(), Item:"Name (by ref)"=REMOVED) -> (then, ReturnValue:bool)
   > Return Node!(execute, ReturnedTags:Array of Names=Get LocalTags(), PartialTagFound:bool=Get LocalPartialTagFound() -> (LocalPartialTagFound:bool))
   n4 = FindSubstring(SearchIn="To String (Name)"(InName=@n3), Substring=Split(SourceString=Get TagText() -> (TagText:String), InStr=:, SearchDir=FromStart))
   n5 = Branch(execute=@n3.LoopBody, Condition=AND Boolean(A="integer >= integer"(A=@n4), B=Get RemoveFoundTags() -> (RemoveFoundTags:bool)))
   > Set LocalPartialTagFound(LocalPartialTagFound=true)
   > Set Array Elem!(execute, TargetArray:Array of Names=Get LocalTags(), Index:int=@n3.Array Index, Item:"Name (by ref)"=REMOVED, bSizeToFit:bool=false) -> (then)
   n6 = FindSubstring(SearchIn="To String (Name)"(InName=@n3))
   Branch(execute=@n5.else, Condition=AND Boolean(A="integer >= integer"(A=@n6), B=Get AllowTagsWithoutColons() -> (AllowTagsWithoutColons:bool)))
   > Set LocalPartialTagFound(LocalPartialTagFound=true)
   Get TagsToSearch() -> (TagsToSearch:Array of Names)
   Get TagsToSearch()
   Get TagsToSearch()
   Get TagsToSearch()
  ]}
 FormatTagValue {nodes [
   n1 = FormatTagValue!() -> (then, Tag:String, ActorObjectReference:Actor Object Reference) {is=title K2Node_FunctionEntry}
   > GetActorClassDefaultObject!(execute, self, Tag:String=@n1, ActorObjectReference:Actor Object Reference=@n1.ActorObjectReference) -> (then)
   > Return Node!(execute, TagResult:String=Get FormattedTag() -> (FormattedTag:String))
  ]}
 GetActorClassDefaultObject {nodes [
   GetActorClassDefaultObject!() -> (then, Tag:String, ActorObjectReference:Actor Object Reference) {is=title K2Node_FunctionEntry}
   > n1 = FindPartialTag!(
    execute
    self
    TagsToSearch:Array of Names=Get Tags!(self:Actor Object Reference=Get ActorObjectReference() -> (ActorObjectReference:Actor Object Reference)) -> (Tags:Array of Names)
    TagText:String=Get Tag() -> (Tag:String)
    RemoveFoundTags:bool=true
    AllowTagsWithoutColons:bool=false
   ) -> (then, ReturnedTags:Array of Names, PartialTagFound:bool) {is=title K2Node_CallFunction}
   n2 = "Get (a copy)"(Array:Array of Names=@n1) -> (Output:Name)
   n3 = "Get (a copy)"(Array=@n1, Dimension 1=1)
   > PrintString(InString=Append!(A="To String (Name)"(InName=@n2), B="\r\n", C:String="To String (Name)"(InName=@n3)) -> (ReturnValue))
   n4 = Get ActorObjectReference()
   > TransactObject(Object=@n4)
   > Set Tags!(execute, Tags:Array of Names=@n1, self:Actor Object Reference=@n4) -> (then, Output_Get:Array of Names)
   n5 = Get Class(Object=@n4)
   n6 = Append!(
    A=Get Tag()
    B=BreakTopLevelAssetPath!(TopLevelAssetPath:"Top Level Asset Path Structure (by ref)"="Get Class Path Name (Top Level Asset Path)"!(Class:Object Class Reference=@n5) -> (ReturnValue:Top Level Asset Path Structure) {is=title K2Node_CallFunction}) -> (PathString:String)
   ) -> (ReturnValue)
   > n7 = Set FormattedTag(FormattedTag:String=@n6) -> (Output_Get:String)
   > PrintString(InString=Append!(A="\r\nTagResult::", B=@n6, C:String="\r\nTagText", D:String) -> (ReturnValue))
   > TransactObject(Object=Get ActorObjectReference())
   > Add Unique(TargetArray:Array of Names=Get Tags(self=Get ActorObjectReference()), NewItem:"Name (by ref)"=String To Name(InString=@n7))
   Set Tags(self=@n4)
  ]}
]
variables [FormattedTag {type=string, default=""}]
widgets {
 info {rootWidgetClass=&Border, widgetCount=3}
 widgets [
  Border_1 {
   widgetClassPath=&Border
   widget=Border_1
   properties {
    verticalAlignment=VAlign_Center
    background {
     bIsDynamicallyLoaded=false
     drawAs=Box
     imageType=NoImage
     mirroring=NoMirror
     resourceName=None
     resourceObject=null
     tiling=NoTile
     imageSize=<32 32>
     margin=<ltrb 0 0 0 0>
     outlineSettings {
      bUseBrushTransparency=false
      roundingType=HalfHeightRadius
      width=0
      color {colorUseRule=UseColor_Specified, specifiedColor=<rgba 0 0 0 0>}
      cornerRadii=<0 0 0 0>
     }
     tintColor {colorUseRule=UseColor_Specified, specifiedColor=<rgba 0.0176 0.0176 0.0176 1>}
     uVRegion {bIsValid=false, max=<0 0>, min=<0 0>}
    }
    padding=<ltrb 4 2 2 2>
   }
  }
  SizeBox_0 {
   widgetClassPath=&SizeBox
   parent=Border_1
   widget=SizeBox_0
   properties {bOverride_HeightOverride=true, bOverride_WidthOverride=true, heightOverride=54, widthOverride=485}
   slot {class=BorderSlot, properties {verticalAlignment=VAlign_Center, padding=<ltrb 4 2 2 2>}}
  }
  TextBlock_86 {
   widgetClassPath=&TextBlock
   parent=SizeBox_0
   widget=TextBlock_86
   properties {
    justification=Center
    text=Automated Function - Class Default Object
    toolTipText=Adds the actors Class Default Object as a Soft Object Path.
    font {
     bForceMonospaced=false
     bMaterialIsStencil=false
     fontMaterial=null
     fontObject=&Roboto
     letterSpacing=0
     monospacedWidth=1
     size=10
     skewAmount=0
     typefaceFontName=Regular
     outlineSettings {
      bApplyOutlineToDropShadows=false
      bMiteredCorners=false
      bSeparateFillAlpha=false
      outlineMaterial=null
      outlineSize=0
      outlineColor=<rgba 0 0 0 1>
     }
    }
   }
   slot {class=SizeBoxSlot, properties {horizontalAlignment=HAlign_Center, verticalAlignment=VAlign_Center, padding=<ltrb 4 2 4 2>}}
  }
 ]
}
