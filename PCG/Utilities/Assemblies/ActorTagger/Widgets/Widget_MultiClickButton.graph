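# Multi-click button UserWidget for the ActorTagger utility, a rounded Border with a text label that distinguishes left, right and middle mouse clicks.
# Mouse-down and hover enter/leave set MouseButtonState, then broadcast BtnEventDispatch with the button itself for listeners to read.
# PreConstruct fills the label from ButtonText, substituting TagText for a placeholder, and DebugOutput enables click logging.
# Tag-related variables like TagMask, TriState and Exclusive are declared but unused in these graphs.
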WidgetBlueprint /PCG/Utilities/Assemblies/ActorTagger/Widgets/Widget_MultiClickButton
graphs [
 BtnEventDispatch {nodes [BtnEventDispatch!() -> (then, Button:Widget Multi Click Button Object Reference) {is=title K2Node_FunctionEntry}]}
 EventGraph {nodes [
   n1 = Click!() -> (OutputDelegate, then, MouseButtonState:int) {is=title K2Node_CustomEvent}
   > n2 = Switch on Int!(execute:Exec, Selection:int=@n1.MouseButtonState) -> (Default:Exec, 0:Exec, 1:Exec, 2:Exec, 3:Exec)
   PrintString!(
    execute=@n2.0
    InString=Mouse Button Not Supported
    bPrintToScreen=false
    bPrintToLog=Get DebugOutput() -> (DebugOutput:bool)
    TextColor
    Duration
    Key
   ) -> (then)
   PrintString(execute=@n2.1, InString=LMB, bPrintToScreen=false, bPrintToLog=Get DebugOutput())
   PrintString(execute=@n2.2, InString=RMB, bPrintToScreen=false, bPrintToLog=Get DebugOutput())
   PrintString(execute=@n2.3, InString=MMB, bPrintToScreen=false, bPrintToLog=Get DebugOutput())
   Event OnMouseEnter()
   > OnMouseHover!(execute, self, Hover State:bool=true) -> (then)
   Event PreConstruct()
   > n3 = Branch(Condition="Equal Exactly (String)"(A=Get ButtonText() -> (ButtonText:String), B=...))
   > n4 = Set ButtonText(ButtonText:String=Get TagText() -> (TagText:String)) -> (Output_Get:String)
   "SetText (Text)"(
    execute=[@n3.else, @n4.then]
    self=Get MultiClickButtonLabel() -> (MultiClickButtonLabel:Text Object Reference)
    InText="To Text (String)"(InString=Get ButtonText())
   )
   Event OnMouseLeave!() -> (OutputDelegate, then, MouseEvent:"Pointer Event Structure (by ref)") {is=title K2Node_Event}
   > OnMouseHover(Hover State=false)
   Event Construct()
  ]}
 OnMouseButtonDown {nodes [
   n1 = OnMouseButtonDown!() -> (then, MyGeometry:Geometry Structure, MouseEvent:"Pointer Event Structure (by ref)") {is=title K2Node_FunctionEntry}
   n2 = Get Effecting Button!(Input:"Pointer Event Structure (by ref)"=@n1.MouseEvent) -> (ReturnValue:Key Structure) {is=title K2Node_CallFunction}
   > n3 = Branch(Condition="Equal (Key)"!(A:Key Structure=@n2, B:Key Structure=LeftMouseButton) -> (ReturnValue:bool))
   > n4 = Set MouseButtonState(MouseButtonState:int=1) -> (Output_Get:int=0)
   n5 = Branch(execute=[@n11.then, @n4.then, @n8.then, @n10.then], Condition=Get DebugOutput())
   > n6 = Click!(execute, self, MouseButtonState:int=Get MouseButtonState() -> (MouseButtonState:int=0)) -> (then) {is=title K2Node_CallFunction}
   Call Btn Event Dispatch!(
    execute:Exec=[@n6.then, @n5.else]
    self:Widget Multi Click Button Object Reference
    Button:Widget Multi Click Button Object Reference=Self-Reference()
   ) -> (then:Exec) {is=title K2Node_CallDelegate}
   > Return Node!(execute, ReturnValue:Event Reply Structure=Handled())
   n7 = Branch(execute=@n3.else, Condition="Equal (Key)"(A=@n2, B=RightMouseButton))
   > n8 = Set MouseButtonState(MouseButtonState=2) -> (Output_Get=0)
   n9 = Branch(execute=@n7.else, Condition="Equal (Key)"(A=@n2, B=MiddleMouseButton))
   > n10 = Set MouseButtonState(MouseButtonState=3) -> (Output_Get=0)
   n11 = Set MouseButtonState(execute=@n9.else, MouseButtonState=0) -> (Output_Get=0)
  ]}
 OnMouseHover {nodes [
   n1 = OnMouseHover!() -> (then, Hover State:bool=~) {is=title K2Node_FunctionEntry}
   > n2 = Branch(Condition=@n1)
   > Set MouseButtonState(MouseButtonState=4) -> (Output_Get=0)
   n3 = Self-Reference()
   > Call Btn Event Dispatch(Button=@n3)
   Set MouseButtonState(execute=@n2.else, MouseButtonState=5) -> (Output_Get=0)
   > Call Btn Event Dispatch(Button=@n3)
  ]}
]
variables [
 BtnEventDispatch {}
 ButtonText {type=string, default=Multi Click}
 DebugOutput {type=boolean}
 Exclusive {type="array<string>", default []}
 MouseButtonState {type=integer}
 PostTagFunction {type=integer}
 SubComponents {type="array<string>", default []}
 TagMask {type=integer}
 TagText {type=string, default=""}
 TriState {type=boolean}
]
widgets {
 info {inheritedWidgetCount=0, parentClass=&UserWidget, rootWidgetClass=&Border, widgetCount=2}
 widgets [
  BorderClickable {
   widgetClassPath=&Border
   widget=BorderClickable
   properties {
    accessibleBehavior=Auto
    horizontalAlignment=HAlign_Center
    verticalAlignment=VAlign_Center
    background {
     bIsDynamicallyLoaded=false
     drawAs=RoundedBox
     imageType=NoImage
     mirroring=NoMirror
     resourceName=None
     resourceObject=null
     tiling=NoTile
     imageSize=<32 32>
     margin=<ltrb 0 0 0 0>
     outlineSettings {
      bUseBrushTransparency=false
      roundingType=FixedRadius
      width=1
      color {colorUseRule=UseColor_ColorTable, specifiedColor=<rgba 0 0 0 0>}
      cornerRadii=<4 4 4 4>
     }
     tintColor {colorUseRule=UseColor_Specified, specifiedColor=<rgba 0.0395 0.0395 0.0395 1>}
     uVRegion {bIsValid=false, max=<0 0>, min=<0 0>}
    }
   }
  }
  MultiClickButtonLabel {
   widgetClassPath=&TextBlock
   bIsVariable=true
   parent=BorderClickable
   widget=MultiClickButtonLabel
   properties {
    accessibleText=ColorAndOpacity
    bCanChildrenBeAccessible=true
    justification=Center
    text=Multi Click
    font {
     bForceMonospaced=false
     bMaterialIsStencil=false
     fontMaterial=null
     fontObject=&Roboto
     letterSpacing=0
     monospacedWidth=1
     size=10
     skewAmount=0
     typefaceFontName=Regular
     outlineSettings {
      bApplyOutlineToDropShadows=false
      bMiteredCorners=false
      bSeparateFillAlpha=false
      outlineMaterial=null
      outlineSize=0
      outlineColor=<rgba 0 0 0 1>
     }
    }
   }
   slot {class=BorderSlot, properties {horizontalAlignment=HAlign_Center, verticalAlignment=VAlign_Center}}
  }
 ]
}
